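; Garage door building actor for the Casino Reaper NPC scenario that slides open on the OpenTheGate event.
; BeginPlay stores the door mesh's location and an end point below it; its trailing Delay leads to nothing.
; OpenTheGate runs once on authority, calling ClientGateSound and playing a timeline that lerps the mesh between those points.
; ClientGateSound plays the opening cue at the start location only on the non-authority branch.
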(event EventBeginPlay
  (bind _doormesh (Variables|Default|GetDoorMesh))
  (bind _returnvalue (Transformation|GetWorldLocation _doormesh))
  (|CallParentFunction)
  (Variables|Default|SetLocation%20at%20Start _returnvalue)
  (Variables|Default|SetLocation%20at%20End (+ _returnvalue "0.000000,0.000000,-368.000000"))
  (Utilities|FlowControl|Delay 12.0))

(event Collision|EventActorBeginOverlap
  (params (OtherActor :type Actor%20Object%20Reference))
  (|CallParentFunction OtherActor) ; This node is disabled and will not be called.\nDrag off pins to build functionality.
)

(event EventTick
  (params (DeltaSeconds :type Float%20%28single-precision%29))
  (|CallParentFunction DeltaSeconds) ; This node is disabled and will not be called.\nDrag off pins to build functionality.
)

(event Custom|OpenTheGate
  (Networking|SwitchHasAuthority
    (:Authority
      (Utilities|FlowControl|DoOnce false
        (:Completed
          (CallFunction|ClientGateSound)
          (bind (_direction _float) (Timeline|Timeline Play)
            (:Update
              (Transformation|SetWorldLocation (Variables|Default|GetDoorMesh) (Math|Vector|Lerp(Vector) (Variables|Default|GetLocation%20at%20Start) (Variables|Default|GetLocation%20at%20End) _float) false false))))))))

(event Custom|ClientGateSound
  (Networking|SwitchHasAuthority
    (:Remote
      (Audio|PlaySoundatLocation "/NPCLibraryCh6/NPCs/CasinoReaper/Audio/CasinoReaper/Quest_S34_Scenario_CasinoReaper_Garage_Open_Cue.Quest_S34_Scenario_CasinoReaper_Garage_Open_Cue" (Variables|Default|GetLocation%20at%20Start)))))

(fn ConstructionScript)
