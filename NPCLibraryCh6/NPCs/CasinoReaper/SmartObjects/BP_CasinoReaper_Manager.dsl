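; Casino Reaper manager that enables a garage door Smart Object when a player starts the Found quest with the NPC.
; After playlist load it enables the patrol point provider and caches the spawned pawn.
; A server-only trigger starts filtered event listeners for non-bot pawns, though end overlap stops them only for bot-controlled pawns.
; The event enables the Smart Object, removes the trigger, pushes AI behavior control and adds gameplay tags.
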(event EventBeginPlay
  (bind _patrolpointprovider (Variables|Default|GetPatrolPointProvider))
  (Class|FortAsyncActionCurrentPlaylistReady|PlaylistDataReadyAsync
    (:then
      (Variables|Default|SetSmartObjectEnabled false)
      ; Trigger should only exist on Server, no need to track Overlaps on Client.
      (Networking|CanExecuteCosmeticEvents
        (:True
          (Components|DestroyComponent (Variables|Default|GetPawnNearAI)))))
    (:OnReady
      ; Enable the Spawn Point so that we can track when the AI has spawned
      (Utilities|FlowControl|DelayUntilNextTick)
      (Class|FortAthenaLivingWorldPointProviderInterface|EnablePointProvider_DEPRECATED _patrolpointprovider)
      (EventDispatchers|BindEventtoOnActorSpawnedResult _patrolpointprovider (EventDispatchers|CreateEvent :!event Custom|CasinoReaperHasSpawned)))))

(event Collision|EventActorBeginOverlap
  (params (OtherActor :type Actor%20Object%20Reference)))

(event EventTick
  (params (DeltaSeconds :type Float%20%28single-precision%29)))

(event Custom|CasinoReaperHasSpawned
  (params (SpawnedActor :type Actor%20Object%20Reference) (bResult :type Boolean))
  ; If we need to do something with the AI, we can do it here.
  (bind _asfort_pawn (Utilities|Casting|CastToFortPawn SpawnedActor))
  (Variables|Default|SetAs%20Fort%20Pawn _asfort_pawn))

(event OnComponentBeginOverlap(PawnNearAI)
  (params (OverlappedComponent :type Primitive%20Component%20Object%20Reference) (OtherActor :type Actor%20Object%20Reference) (OtherComp :type Primitive%20Component%20Object%20Reference) (OtherBodyIndex :type Integer) (bFromSweep :type Boolean) (SweepResult :type Hit%20Result%20Structure%20%28by%20ref%29))
  ; When a Pawn gets close, if track if he chooses the option to start the Found Quest
  (bind _asfort_pawn (Utilities|Casting|CastToFortPawn OtherActor))
  (if (Pawn|IsBotControlled _asfort_pawn)
    _
    (elif (Variables|Default|GetSmartObjectEnabled)
      _
      (else
        (bind _returnvalue (EventProcessorSubsystem|StartListeningToFilteredEvent 0 OtherActor (Variables|Default|GetVerb%20Filter) false (EventDispatchers|CreateEvent :!event Custom|AIListener)))
        (Utilities|Map|Add (Variables|Default|GetPawnToEventHandles) OtherActor _returnvalue)))))

(event OnComponentEndOverlap(PawnNearAI)
  (params (OverlappedComponent :type Primitive%20Component%20Object%20Reference) (OtherActor :type Actor%20Object%20Reference) (OtherComp :type Primitive%20Component%20Object%20Reference) (OtherBodyIndex :type Integer))
  (bind _pawntoeventhandles (Variables|Default|GetPawnToEventHandles))
  ; If the player moves away, stop tracking the Conversations with the AI and it's options.
  (bind _asfort_pawn (Utilities|Casting|CastToFortPawn OtherActor))
  (if (Pawn|IsBotControlled _asfort_pawn)
    (if (.ReturnValue (Utilities|Map|Find _pawntoeventhandles _asfort_pawn))
      (EventProcessorSubsystem|StopListeningToFilteredEvent (Utilities|Map|Find _pawntoeventhandles _asfort_pawn)))))

(event Custom|AIListener
  (bind _as_fort_pawn (Variables|Default|GetAs%20Fort%20Pawn))
  ; WHen someone chooses to start the Found Quest Enable the Smart Object
  (if (Variables|Default|GetSmartObjectEnabled)
    _
    (elif (SmartObject|SetSmartObjectEnabled(defaultreason:Gameplay) (Actor|GetComponentbyClass (Variables|Default|GetGarageDoor) "/Script/SmartObjectsModule.SmartObjectComponent") true :!class /Script/SmartObjectsModule.SmartObjectComponent)
      (Variables|Default|SetSmartObjectEnabled true)
      (Components|DestroyComponent (Variables|Default|GetPawnNearAI))
      (bind _asfort_athena_aibot_controller (Utilities|Casting|CastToFortAthenaAIBotController (Pawn|GetController _as_fort_pawn)))
      (AI|PushBehaviorControl _asfort_athena_aibot_controller -1056769)
      (Ability|GameplayEffect|AddLooseGameplayTags _as_fort_pawn (Variables|Default|GetGameplay%20Tags)))))

(fn ConstructionScript)
