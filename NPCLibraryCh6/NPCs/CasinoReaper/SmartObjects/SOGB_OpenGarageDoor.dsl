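; Smart Object gameplay behavior that lets a character, likely an AI NPC, open the Casino Reaper garage door.
; On trigger it calls OpentheGate on the owning door, waits three seconds, adds gameplay tags to the pawn, disables the owner's SmartObject components and ends.
; Init, SetFocalPoint and ResetFocalPoint handle WaitAndLook config, bot controller caching and look direction, but nothing shown calls them.
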(event GameplayBehavior|EventOnTriggeredCharacter
  (params (Avatar :type Character%20Object%20Reference) (Config :type Gameplay%20Behavior%20Config%20Object%20Reference) (SmartObjectOwner :type Actor%20Object%20Reference))
  (bind _asbga_garage_door_casino_reaper (Utilities|Casting|CastToBGA_GarageDoor_CasinoReaper SmartObjectOwner))
  (Class|BGAGarageDoorCasinoReaper|OpentheGate _asbga_garage_door_casino_reaper)
  (GameplayTasks|TaskWaitDelay 0 3.0
    (:OnFinish
      (bind _asfort_player_pawn (Utilities|Casting|CastToFortPlayerPawn Avatar))
      (Ability|GameplayEffect|AddLooseGameplayTags _asfort_player_pawn (Variables|Default|GetGameplay%20Tags) true)
      (for _array_element (Actor|GetComponentsByClass SmartObjectOwner "/Script/SmartObjectsModule.SmartObjectComponent")
        (if (SmartObject|SetSmartObjectEnabled(defaultreason:Gameplay) _array_element false :!class /Script/SmartObjectsModule.SmartObjectComponent)
          _))
      (GameplayBehavior|EndBehavior (Variables|Default|GetUserAvatar)))))

(event GameplayBehavior|EventOnFinishedCharacter
  (params (Avatar :type Character%20Object%20Reference) (bWasInterrupted :type Boolean)))

(fn Init
  (params (Avatar :type Character%20Object%20Reference) (BehaviorConfig :type Gameplay%20Behavior%20Config%20Object%20Reference) (SmartObjectOwner :type Actor%20Object%20Reference))
  (Utilities|FlowControl|Sequence
    (:then_0
      (bind _asfort_player_pawn (Utilities|Casting|CastToFortPlayerPawn Avatar)
        (:then
          (bind _output_get (Variables|Default|SetUserPawn _asfort_player_pawn))
          (bind _asfort_athena_aibot_controller (Utilities|Casting|CastToFortAthenaAIBotController (Pawn|GetController _output_get))
            (:then
              (Variables|Default|SetUserController _asfort_athena_aibot_controller))
            (:CastFailed
              _)))
        (:CastFailed
          _))
      (LocalVariables|SetSuccessfullyInitialized false))
    (:then_1
      (bind _assogb_wait_and_look_config (Utilities|Casting|CastToSOGB_WaitAndLook_Config BehaviorConfig)
        (:then
          (Variables|Default|SetWaitAndLookConfig _assogb_wait_and_look_config))
        (:CastFailed
          (LocalVariables|SetSuccessfullyInitialized false))))
    (:then_2
      (Variables|Default|SetSmartObjectOwner SmartObjectOwner))
    (:then_3
      (return (LocalVariables|GetSuccessfullyInitialized)))))

(fn ResetFocalPoint
  (bind _usercontroller (Variables|Default|GetUserController)
    (:then
      (AI|ClearFocus _usercontroller))
    (:else)))

(fn SetFocalPoint
  (if (Math|Boolean|ANDBoolean (Utilities|IsValid (Variables|Default|GetUserPawn) :!class /Script/Engine.KismetSystemLibrary) (Utilities|IsValid (Variables|Default|GetUserController) :!class /Script/Engine.KismetSystemLibrary) (Utilities|IsValid (Variables|Default|GetSmartObjectOwner) :!class /Script/Engine.KismetSystemLibrary))
    (AI|SetFocalPoint (Variables|Default|GetUserController) (+ (Transformation|GetActorLocation (Variables|Default|GetUserPawn)) (* (Transformation|GetActorForwardVector (Variables|Default|GetSmartObjectOwner)) 1e+02)))))
